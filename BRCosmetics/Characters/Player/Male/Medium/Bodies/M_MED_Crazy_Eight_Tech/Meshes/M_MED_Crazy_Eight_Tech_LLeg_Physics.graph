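# Physics asset for the left leg of the male medium Crazy Eight Tech body cosmetic, covering a dangling leg pocket and strap.
# Capsule bodies cover six strap bones and the pocket, plus larger pelvis and thigh shapes approximating the body.
# Straps 02 to 06 form a chain with tight angular limits, loosening slightly toward the tip.
# Free-rotation constraints link thigh, pocket and strap 01; no constraint joins straps 01 and 02.
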PhysicsAsset /BRCosmetics/Characters/Player/Male/Medium/Bodies/M_MED_Crazy_Eight_Tech/Meshes/M_MED_Crazy_Eight_Tech_LLeg_Physics
bodies {
 dyn_leg_pocket [{shapeType=Capsule, length=2, radius=2, shapeName=dyn_leg_pocket_capsule}]
 dyn_strap_01_l [{shapeType=Capsule, length=2, radius=2, shapeName=dyn_strap_01_l_capsule, center=<4.0344 0 -0.0001>}]
 dyn_strap_02_l [{shapeType=Capsule, length=2, radius=2, shapeName=dyn_strap_02_l_capsule, center=<4.2093 0 0.0001>}]
 dyn_strap_03_l [{shapeType=Capsule, length=3.0618, radius=2, shapeName=dyn_strap_03_l_capsule, center=<2 0 0>}]
 dyn_strap_04_l [{shapeType=Capsule, length=3.0618, radius=2, shapeName=dyn_strap_04_l_capsule, center=<2 0 0>}]
 dyn_strap_05_l [{shapeType=Capsule, length=3.0618, radius=2, shapeName=dyn_strap_05_l_capsule, center=<2 0 0>}]
 dyn_strap_06_l [{shapeType=Capsule, length=3.5, radius=2, shapeName=dyn_strap_06_l_capsule, center=<2 0 0>}]
 pelvis [
  {shapeType=Capsule, length=9.6542, radius=8.6662, shapeName=pelvis_capsule, center=<-4.6674 2.0531 -0.0001>}
  {shapeType=Capsule, length=9.3507, radius=10.9321, shapeName=pelvis_capsule, center=<-1.2883 -0.8557 0.0001>}
 ]
 thigh_l [{shapeType=Capsule, length=19.607, radius=8.3305, shapeName=thigh_l_capsule, center=<-18.398 -0.5209 0>, rotation=<rot -90 0 -5.0002>}]
}
constraints [
 {bone1Name=dyn_strap_03_l, bone2Name=dyn_strap_02_l, swing1LimitDegrees=20, swing2LimitDegrees=3, twistLimitDegrees=2}
 {bone1Name=dyn_strap_04_l, bone2Name=dyn_strap_03_l, swing1LimitDegrees=20, swing2LimitDegrees=5, twistLimitDegrees=2}
 {bone1Name=dyn_strap_05_l, bone2Name=dyn_strap_04_l, swing1LimitDegrees=20, swing2LimitDegrees=10, twistLimitDegrees=2}
 {bone1Name=dyn_strap_06_l, bone2Name=dyn_strap_05_l, swing1LimitDegrees=20, swing2LimitDegrees=10, twistLimitDegrees=2}
 {
  bone1Name=dyn_leg_pocket
  bone2Name=dyn_strap_01_l
  swing1LimitDegrees=45
  swing1Motion=Free
  swing2LimitDegrees=45
  swing2Motion=Free
  twistLimitDegrees=45
  twistMotion=Free
 }
 {
  bone1Name=thigh_l
  bone2Name=dyn_leg_pocket
  swing1LimitDegrees=45
  swing1Motion=Free
  swing2LimitDegrees=45
  swing2Motion=Free
  twistLimitDegrees=45
  twistMotion=Free
 }
]
