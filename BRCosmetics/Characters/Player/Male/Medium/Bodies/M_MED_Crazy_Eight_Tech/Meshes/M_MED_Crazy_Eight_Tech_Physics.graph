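# Physics asset for the M_MED_Crazy_Eight_Tech male player body, covering its dangling front and back strap bones, with capsules on pelvis and thighs.
# Back straps form two tightly limited six-bone chains hung from the pelvis, with matching left and right bones cross-linked by angularly free constraints.
# Front straps form narrow-limit chains from segment 02 down, one segment longer on the right.
# Only right front segment 01 is constrained to the pelvis, and neither 01-to-02 link has a constraint.
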PhysicsAsset /BRCosmetics/Characters/Player/Male/Medium/Bodies/M_MED_Crazy_Eight_Tech/Meshes/M_MED_Crazy_Eight_Tech_Physics
bodies {
 dyn_back_strap_01_l [{shapeType=Capsule, length=6.4447, radius=1.25, shapeName=dyn_back_strap_01_l_capsule, center=<5.9351 0 0.3792>}]
 dyn_back_strap_01_r [{shapeType=Capsule, length=6.4447, radius=1.25, shapeName=dyn_back_strap_01_l_capsule, center=<-5.9351 0 -0.3792>, rotation=<rot 0 0 180>}]
 dyn_back_strap_02_l [{shapeType=Box, extentX=9.2134, extentY=0.9792, extentZ=8.1003, shapeName=dyn_back_strap_02_l_box, center=<4.6972 -0.2565 0.2352>}]
 dyn_back_strap_02_r [{
   shapeType=Box
   extentX=9.2134
   extentY=0.9792
   extentZ=8.1003
   shapeName=dyn_back_strap_02_l_box
   center=<-4.6972 0.2565 -0.2352>
   rotation=<rot 0 0 180>
  }]
 dyn_back_strap_03_l [{shapeType=Capsule, length=6.4447, radius=1.25, shapeName=dyn_back_strap_03_l_capsule, center=<9.1158 0 0.3793>}]
 dyn_back_strap_03_r [{shapeType=Capsule, length=6.4447, radius=1.25, shapeName=dyn_back_strap_03_l_capsule, center=<-9.1158 0 -0.3793>, rotation=<rot 0 0 180>}]
 dyn_back_strap_04_l [{shapeType=Capsule, length=6.4447, radius=1.25, shapeName=dyn_back_strap_04_l_capsule, center=<9.1158 0 0.3793>}]
 dyn_back_strap_04_r [{shapeType=Capsule, length=6.4447, radius=1.25, shapeName=dyn_back_strap_04_l_capsule, center=<-9.1158 0 -0.3793>, rotation=<rot 0 0 180>}]
 dyn_back_strap_05_l [{shapeType=Capsule, length=5.6453, radius=1.25, shapeName=dyn_back_strap_05_l_capsule, center=<9.1158 0 0.3793>}]
 dyn_back_strap_05_r [{shapeType=Capsule, length=5.6453, radius=1.25, shapeName=dyn_back_strap_05_l_capsule, center=<-9.1158 0 -0.3793>, rotation=<rot 0 0 180>}]
 dyn_back_strap_06_l [{shapeType=Capsule, length=4.7898, radius=1.25, shapeName=dyn_back_strap_06_l_capsule, center=<9.1158 -0.0001 -0.062>}]
 dyn_back_strap_06_r [{shapeType=Capsule, length=4.7898, radius=1.25, shapeName=dyn_back_strap_06_l_capsule, center=<-9.1158 0.0001 0.062>, rotation=<rot 0 0 180>}]
 dyn_front_strap_01_l [{shapeType=Capsule, length=2, radius=1.25, shapeName=dyn_front_strap_01_l_capsule, center=<6.8847 0.0002 0>}]
 dyn_front_strap_01_r [{shapeType=Capsule, length=9.0565, radius=1.25, shapeName=dyn_front_strap_01_r_capsule, center=<6.8263 0.0001 0>}]
 dyn_front_strap_02_l [{shapeType=Capsule, length=9.1221, radius=1.75, shapeName=dyn_front_strap_02_l_capsule, center=<3.7587 0 0>}]
 dyn_front_strap_02_r [{shapeType=Capsule, length=7.6301, radius=1.75, shapeName=dyn_front_strap_02_r_capsule, center=<3.87 0.0001 0>}]
 dyn_front_strap_03_l [{shapeType=Capsule, length=5.5, radius=1.75, shapeName=dyn_front_strap_03_l_capsule, center=<4.9551 -0.0001 0>}]
 dyn_front_strap_03_r [{shapeType=Capsule, length=6.7566, radius=1.75, shapeName=dyn_front_strap_03_r_capsule, center=<4.3359 0.0001 0>}]
 dyn_front_strap_04_l [{shapeType=Capsule, length=4, radius=1.75, shapeName=dyn_front_strap_04_l_capsule, center=<5.7237 0 0>}]
 dyn_front_strap_04_r [{shapeType=Capsule, length=4.9363, radius=1.75, shapeName=dyn_front_strap_04_r_capsule, center=<6.1368 0.0001 0>}]
 dyn_front_strap_05_l [{shapeType=Capsule, length=2.9286, radius=1.75, shapeName=dyn_front_strap_05_l_capsule, center=<6.6333 0.0001 0.1196>}]
 dyn_front_strap_05_r [{shapeType=Capsule, length=4.1378, radius=1.75, shapeName=dyn_front_strap_05_r_capsule, center=<4.6018 0.0001 0>}]
 dyn_front_strap_06_l [{shapeType=Capsule, length=3.5, radius=1.75, shapeName=dyn_front_strap_06_l_capsule, center=<7.4539 0 0>}]
 dyn_front_strap_06_r [{shapeType=Capsule, length=4.2974, radius=1.75, shapeName=dyn_front_strap_06_r_capsule, center=<5.5052 0.0001 0>}]
 dyn_front_strap_07_l [{shapeType=Capsule, length=3.2, radius=1.75, shapeName=dyn_front_strap_07_l_capsule, center=<6.6127 0 0>}]
 dyn_front_strap_07_r [{shapeType=Capsule, length=4.6969, radius=1.75, shapeName=dyn_front_strap_07_r_capsule, center=<5.8912 0.0001 0>}]
 dyn_front_strap_08_l [{shapeType=Capsule, length=2.3803, radius=1.75, shapeName=dyn_front_strap_08_l_capsule, center=<5.9108 -0.0002 0.0004>}]
 dyn_front_strap_08_r [{shapeType=Capsule, length=2.2783, radius=1.75, shapeName=dyn_front_strap_08_r_capsule, center=<10.0305 0.0001 0.0001>}]
 dyn_front_strap_09_r [{shapeType=Capsule, length=2.0492, radius=1.75, shapeName=dyn_front_strap_09_r_capsule, center=<0.2426 0 -0.0002>}]
 pelvis [
  {shapeType=Capsule, length=9.6542, radius=8.6662, shapeName=pelvis_capsule, center=<-4.6674 2.0531 -0.0001>}
  {shapeType=Capsule, length=9.3507, radius=10.9321, shapeName=pelvis_capsule, center=<-1.2883 -0.8557 0.0001>}
 ]
 thigh_l [{shapeType=Capsule, length=27.232, radius=9.5805, shapeName=thigh_l_capsule, center=<-18.398 -0.5209 0>, rotation=<rot -90 0 -5.0002>}]
 thigh_r [{shapeType=Capsule, length=27.232, radius=9.5805, shapeName=thigh_l_capsule, center=<18.398 -0.5209 0>, rotation=<rot -90 0 -5.0002>}]
}
constraints [
 {bone1Name=dyn_back_strap_01_l, bone2Name=pelvis, swing1LimitDegrees=12, swing2LimitDegrees=12, twistLimitDegrees=5}
 {bone1Name=dyn_back_strap_02_l, bone2Name=dyn_back_strap_01_l, swing1LimitDegrees=12, swing2LimitDegrees=12, twistLimitDegrees=2.5}
 {bone1Name=dyn_back_strap_03_l, bone2Name=dyn_back_strap_02_l, swing1LimitDegrees=12, swing2LimitDegrees=12, twistLimitDegrees=2.5}
 {bone1Name=dyn_back_strap_04_l, bone2Name=dyn_back_strap_03_l, swing1LimitDegrees=12, swing2LimitDegrees=12, twistLimitDegrees=2.5}
 {bone1Name=dyn_back_strap_05_l, bone2Name=dyn_back_strap_04_l, swing1LimitDegrees=12, swing2LimitDegrees=12, twistLimitDegrees=2.5}
 {bone1Name=dyn_back_strap_06_l, bone2Name=dyn_back_strap_05_l, swing1LimitDegrees=12, swing2LimitDegrees=12, twistLimitDegrees=2.5}
 {bone1Name=dyn_back_strap_01_r, bone2Name=pelvis, swing1LimitDegrees=12, swing2LimitDegrees=12, twistLimitDegrees=2.5}
 {bone1Name=dyn_back_strap_02_r, bone2Name=dyn_back_strap_01_r, swing1LimitDegrees=12, swing2LimitDegrees=12, twistLimitDegrees=2.5}
 {bone1Name=dyn_back_strap_03_r, bone2Name=dyn_back_strap_02_r, swing1LimitDegrees=12, swing2LimitDegrees=12, twistLimitDegrees=2.5}
 {bone1Name=dyn_back_strap_04_r, bone2Name=dyn_back_strap_03_r, swing1LimitDegrees=12, swing2LimitDegrees=12, twistLimitDegrees=2.5}
 {bone1Name=dyn_back_strap_05_r, bone2Name=dyn_back_strap_04_r, swing1LimitDegrees=12, swing2LimitDegrees=12, twistLimitDegrees=2.5}
 {bone1Name=dyn_back_strap_06_r, bone2Name=dyn_back_strap_05_r, swing1LimitDegrees=12, swing2LimitDegrees=12, twistLimitDegrees=2.5}
 {
  bone1Name=thigh_r
  bone2Name=pelvis
  swing1LimitDegrees=45
  swing1Motion=Free
  swing2LimitDegrees=45
  swing2Motion=Free
  twistLimitDegrees=45
  twistMotion=Free
 }
 {bone1Name=dyn_front_strap_01_r, bone2Name=pelvis, swing1LimitDegrees=20, swing2LimitDegrees=2, twistLimitDegrees=2}
 {
  bone1Name=dyn_back_strap_01_l
  bone2Name=dyn_back_strap_01_r
  swing1LimitDegrees=45
  swing1Motion=Free
  swing2LimitDegrees=45
  swing2Motion=Free
  twistLimitDegrees=45
  twistMotion=Free
 }
 {
  bone1Name=dyn_back_strap_02_l
  bone2Name=dyn_back_strap_02_r
  swing1LimitDegrees=45
  swing1Motion=Free
  swing2LimitDegrees=45
  swing2Motion=Free
  twistLimitDegrees=45
  twistMotion=Free
 }
 {
  bone1Name=dyn_back_strap_03_l
  bone2Name=dyn_back_strap_03_r
  swing1LimitDegrees=45
  swing1Motion=Free
  swing2LimitDegrees=45
  swing2Motion=Free
  twistLimitDegrees=45
  twistMotion=Free
 }
 {
  bone1Name=dyn_back_strap_04_l
  bone2Name=dyn_back_strap_04_r
  swing1LimitDegrees=45
  swing1Motion=Free
  swing2LimitDegrees=45
  swing2Motion=Free
  twistLimitDegrees=45
  twistMotion=Free
 }
 {
  bone1Name=dyn_back_strap_05_l
  bone2Name=dyn_back_strap_05_r
  swing1LimitDegrees=45
  swing1Motion=Free
  swing2LimitDegrees=45
  swing2Motion=Free
  twistLimitDegrees=45
  twistMotion=Free
 }
 {
  bone1Name=dyn_back_strap_06_l
  bone2Name=dyn_back_strap_06_r
  swing1LimitDegrees=45
  swing1Motion=Free
  swing2LimitDegrees=45
  swing2Motion=Free
  twistLimitDegrees=45
  twistMotion=Free
 }
 {bone1Name=dyn_front_strap_03_r, bone2Name=dyn_front_strap_02_r, swing1LimitDegrees=10, swing2LimitDegrees=10, twistLimitDegrees=1}
 {bone1Name=dyn_front_strap_04_r, bone2Name=dyn_front_strap_03_r, swing1LimitDegrees=10, swing2LimitDegrees=10, twistLimitDegrees=1}
 {bone1Name=dyn_front_strap_05_r, bone2Name=dyn_front_strap_04_r, swing1LimitDegrees=10, swing2LimitDegrees=10, twistLimitDegrees=1}
 {bone1Name=dyn_front_strap_06_r, bone2Name=dyn_front_strap_05_r, swing1LimitDegrees=10, swing2LimitDegrees=10, twistLimitDegrees=1}
 {bone1Name=dyn_front_strap_07_r, bone2Name=dyn_front_strap_06_r, swing1LimitDegrees=10, swing2LimitDegrees=10, twistLimitDegrees=1}
 {bone1Name=dyn_front_strap_08_r, bone2Name=dyn_front_strap_07_r, swing1LimitDegrees=10, swing2LimitDegrees=10, twistLimitDegrees=1}
 {bone1Name=dyn_front_strap_09_r, bone2Name=dyn_front_strap_08_r, swing1LimitDegrees=10, swing2LimitDegrees=10, twistLimitDegrees=1}
 {bone1Name=dyn_front_strap_03_l, bone2Name=dyn_front_strap_02_l, swing1LimitDegrees=10, swing2LimitDegrees=10, twistLimitDegrees=1}
 {bone1Name=dyn_front_strap_04_l, bone2Name=dyn_front_strap_03_l, swing1LimitDegrees=10, swing2LimitDegrees=10, twistLimitDegrees=1}
 {bone1Name=dyn_front_strap_05_l, bone2Name=dyn_front_strap_04_l, swing1LimitDegrees=10, swing2LimitDegrees=10, twistLimitDegrees=1}
 {bone1Name=dyn_front_strap_06_l, bone2Name=dyn_front_strap_05_l, swing1LimitDegrees=10, swing2LimitDegrees=10, twistLimitDegrees=1}
 {bone1Name=dyn_front_strap_07_l, bone2Name=dyn_front_strap_06_l, swing1LimitDegrees=10, swing2LimitDegrees=10, twistLimitDegrees=1}
 {bone1Name=dyn_front_strap_08_l, bone2Name=dyn_front_strap_07_l, swing1LimitDegrees=10, swing2LimitDegrees=10, twistLimitDegrees=1}
]
