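# Physics asset for the right leg of the Crazy Eight Tech medium male player body, covering a dangling leg strap and the body around it.
# Capsule bodies cover the pelvis, right thigh and calf, plus a strap master and five-segment strap chain hanging off the thigh.
# Constraints down to the first strap segment, and at the knee, rotate freely, while the later strap links have tight swing and twist limits.
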PhysicsAsset /BRCosmetics/Characters/Player/Male/Medium/Bodies/M_MED_Crazy_Eight_Tech/Meshes/M_MED_Crazy_Eight_Tech_RLeg_Physics
bodies {
 calf_r [{shapeType=Capsule, length=23.0943, radius=6.0823, shapeName=calf_r_capsule, center=<14.8562 0.8328 0.5056>, rotation=<rot 90 0 0>}]
 dyn_leg_strap_01_r [{
   shapeType=Capsule
   length=4.9371
   radius=4.2612
   shapeName=dyn_leg_strap_01_r_capsule
   center=<1.0618 0.3974 -0.6696>
   rotation=<rot -6.8423 -90 7.371>
  }]
 dyn_leg_strap_02_r [{shapeType=Capsule, length=2.8118, radius=2, shapeName=dyn_strap_03_l_capsule, center=<2 0 0>}]
 dyn_leg_strap_03_r [{shapeType=Capsule, length=2.8118, radius=2, shapeName=dyn_strap_03_l_capsule, center=<2 0 0>}]
 dyn_leg_strap_04_r [{shapeType=Capsule, length=2.8118, radius=2, shapeName=dyn_strap_03_l_capsule, center=<2 0 0>}]
 dyn_leg_strap_05_r [{shapeType=Capsule, length=2.8118, radius=2, shapeName=dyn_strap_03_l_capsule, center=<2 0 0>}]
 dyn_leg_strap_mstr_r [{shapeType=Capsule, length=0.505, radius=0.505, shapeName=dyn_leg_strap_mstr_r_capsule, rotation=<rot 0 -90 0>}]
 pelvis [
  {shapeType=Capsule, length=9.6542, radius=8.6662, shapeName=pelvis_capsule, center=<-4.6674 2.0531 -0.0001>}
  {shapeType=Capsule, length=9.3507, radius=10.9321, shapeName=pelvis_capsule, center=<-1.2883 -0.8557 0.0001>}
 ]
 thigh_r [{shapeType=Capsule, length=22.482, radius=7.518, shapeName=thigh_l_capsule, center=<18.398 -0.5209 0>, rotation=<rot -90 0 -5.0002>}]
}
constraints [
 {
  bone1Name=thigh_r
  bone2Name=pelvis
  swing1LimitDegrees=45
  swing1Motion=Free
  swing2LimitDegrees=45
  swing2Motion=Free
  twistLimitDegrees=45
  twistMotion=Free
 }
 {
  bone1Name=dyn_leg_strap_mstr_r
  bone2Name=thigh_r
  swing1LimitDegrees=45
  swing1Motion=Free
  swing2LimitDegrees=45
  swing2Motion=Free
  twistLimitDegrees=45
  twistMotion=Free
 }
 {
  bone1Name=dyn_leg_strap_01_r
  bone2Name=dyn_leg_strap_mstr_r
  swing1LimitDegrees=45
  swing1Motion=Free
  swing2LimitDegrees=45
  swing2Motion=Free
  twistLimitDegrees=45
  twistMotion=Free
 }
 {bone1Name=dyn_leg_strap_02_r, bone2Name=dyn_leg_strap_01_r, swing1LimitDegrees=20, swing2LimitDegrees=10, twistLimitDegrees=2}
 {bone1Name=dyn_leg_strap_03_r, bone2Name=dyn_leg_strap_02_r, swing1LimitDegrees=20, swing2LimitDegrees=10, twistLimitDegrees=2}
 {bone1Name=dyn_leg_strap_04_r, bone2Name=dyn_leg_strap_03_r, swing1LimitDegrees=20, swing2LimitDegrees=10, twistLimitDegrees=2}
 {bone1Name=dyn_leg_strap_05_r, bone2Name=dyn_leg_strap_04_r, swing1LimitDegrees=20, swing2LimitDegrees=10, twistLimitDegrees=2}
 {
  bone1Name=calf_r
  bone2Name=thigh_r
  swing1LimitDegrees=45
  swing1Motion=Free
  swing2LimitDegrees=45
  swing2Motion=Free
  twistLimitDegrees=45
  twistMotion=Free
 }
]
