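# Physics asset for the torso mesh of the male medium Crazy Eight Tech player cosmetic, with dangling arm straps and zipper pulls.
# Capsules cover pelvis, spine, clavicles and upper arms, chained by 45-degree limited constraints.
# Small MSTR bodies attach to the clavicles with free rotation and anchor the dynamic chains.
# Two-segment straps and single-segment zippers hang from these anchors under tighter swing and narrow twist limits.
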PhysicsAsset /BRCosmetics/Characters/Player/Male/Medium/Bodies/M_MED_Crazy_Eight_Tech/Meshes/M_MED_Crazy_Eight_Tech_Torso_Physics
bodies {
 clavicle_l [{shapeType=Capsule, length=16.45, radius=4, center=<20.8868 -2.2893 -0.3558>, rotation=<rot 0 -107.3695 -64.2241>}]
 clavicle_r [{shapeType=Capsule, length=16.4459, radius=4, center=<-20.8459 2.4338 0.3374>, rotation=<rot 0 -107.4679 -64.4008>}]
 dyn_arm_strap_01_l [{
   shapeType=Capsule
   length=1.3004
   radius=1.2145
   shapeName=dyn_arm_strap_01_l_capsule
   center=<1.1123 0 0>
   rotation=<rot -0.2222 -90 -6.499>
  }]
 dyn_arm_strap_01_l_MSTR [{shapeType=Capsule, length=0.505, radius=0.505, shapeName=dyn_arm_strap_01_l_MSTR_capsule, rotation=<rot 0 -90 0>}]
 dyn_arm_strap_01_r [{
   shapeType=Capsule
   length=1.2936
   radius=1.1787
   shapeName=dyn_arm_strap_01_r_capsule
   center=<1.1421 0 0>
   rotation=<rot -0.1507 -90 6.1943>
  }]
 dyn_arm_strap_01_r_MSTR [{shapeType=Capsule, length=0.505, radius=0.505, shapeName=dyn_arm_strap_01_r_MSTR_capsule, rotation=<rot 0 -90 0>}]
 dyn_arm_strap_02_l [{
   shapeType=Capsule
   length=1.7394
   radius=1.1407
   shapeName=dyn_arm_strap_02_l_capsule
   center=<0.9019 0 0>
   rotation=<rot 0.9959 -90 -2.9903>
  }]
 dyn_arm_strap_02_r [{shapeType=Capsule, length=1.7394, radius=1.1422, shapeName=dyn_arm_strap_02_r_capsule, center=<0.8876 0 0>, rotation=<rot 0.0379 -90 2.6021>}]
 dyn_torso_zipper_l [{shapeType=Capsule, length=1.6615, radius=0.5849, shapeName=dyn_torso_zipper_l_capsule, center=<1.5 0 0>, rotation=<rot 90 0 0>}]
 dyn_torso_zipper_l_MSTR [{shapeType=Capsule, length=0.505, radius=0.505, shapeName=dyn_torso_zipper_l_MSTR_capsule, rotation=<rot 0 -90 0>}]
 dyn_torso_zipper_r [{shapeType=Capsule, length=1.6615, radius=0.5849, shapeName=dyn_torso_zipper_r_capsule, center=<1.5 0 0>, rotation=<rot 90 0 0>}]
 dyn_torso_zipper_r_MSTR [{shapeType=Capsule, length=0.505, radius=0.505, shapeName=dyn_torso_zipper_r_MSTR_capsule, rotation=<rot 0 -90 0>}]
 pelvis [{shapeType=Capsule, length=15.9339, radius=15.2406, center=<-0.8012 -2.7503 0.1062>, rotation=<rot 0 -25.02 -49.7475>}]
 spine_01 [{shapeType=Capsule, length=17.1247, radius=13.2523, center=<1.6052 -1.1042 -0.7368>, rotation=<rot -14.6611 -90 -17.2212>}]
 spine_02 [{shapeType=Capsule, length=16.6097, radius=11.0568, center=<3.6566 1.9464 0.1594>, rotation=<rot -11.0616 -90 0.0603>}]
 spine_03 [{shapeType=Capsule, length=17.0087, radius=15.2593, center=<3.9695 -0.3892 -0.2101>, rotation=<rot -11.4627 -90 2.1089>}]
 spine_04 [{shapeType=Capsule, length=18.222, radius=15.1487, center=<2.3661 0.2081 0.5463>, rotation=<rot 0 0.1228 -177.1149>}]
 spine_05 [{shapeType=Capsule, length=14.2936, radius=14.0793, center=<1.7919 1.0229 0.4328>, rotation=<rot 0 21.4192 -87.0816>}]
 upperarm_l [{shapeType=Capsule, length=25.9945, radius=7, center=<22.1454 4.3672 -0.4412>, rotation=<rot 0 -78.3086 -85.6977>}]
 upperarm_r [{shapeType=Capsule, length=25.9901, radius=7, center=<-22.1174 -4.3489 0.9624>, rotation=<rot 0 -78.3634 -85.7731>}]
}
constraints [
 {bone1Name=spine_01, bone2Name=pelvis, swing1LimitDegrees=45, swing2LimitDegrees=45, twistLimitDegrees=45}
 {bone1Name=spine_02, bone2Name=spine_01, swing1LimitDegrees=45, swing2LimitDegrees=45, twistLimitDegrees=45}
 {bone1Name=spine_03, bone2Name=spine_02, swing1LimitDegrees=45, swing2LimitDegrees=45, twistLimitDegrees=45}
 {bone1Name=spine_04, bone2Name=spine_03, swing1LimitDegrees=45, swing2LimitDegrees=45, twistLimitDegrees=45}
 {bone1Name=spine_05, bone2Name=spine_04, swing1LimitDegrees=45, swing2LimitDegrees=45, twistLimitDegrees=45}
 {bone1Name=clavicle_l, bone2Name=spine_05, swing1LimitDegrees=45, swing2LimitDegrees=45, twistLimitDegrees=45}
 {bone1Name=upperarm_l, bone2Name=clavicle_l, swing1LimitDegrees=45, swing2LimitDegrees=45, twistLimitDegrees=45}
 {bone1Name=clavicle_r, bone2Name=spine_05, swing1LimitDegrees=45, swing2LimitDegrees=45, twistLimitDegrees=45}
 {bone1Name=upperarm_r, bone2Name=clavicle_r, swing1LimitDegrees=45, swing2LimitDegrees=45, twistLimitDegrees=45}
 {
  bone1Name=dyn_arm_strap_01_l_MSTR
  bone2Name=clavicle_l
  swing1LimitDegrees=45
  swing1Motion=Free
  swing2LimitDegrees=45
  swing2Motion=Free
  twistLimitDegrees=45
  twistMotion=Free
 }
 {bone1Name=dyn_arm_strap_01_l, bone2Name=dyn_arm_strap_01_l_MSTR, swing1LimitDegrees=25, swing2LimitDegrees=25, twistLimitDegrees=5}
 {bone1Name=dyn_arm_strap_02_l, bone2Name=dyn_arm_strap_01_l, swing1LimitDegrees=25, swing2LimitDegrees=25, twistLimitDegrees=5}
 {
  bone1Name=dyn_arm_strap_01_r_MSTR
  bone2Name=clavicle_r
  swing1LimitDegrees=45
  swing1Motion=Free
  swing2LimitDegrees=45
  swing2Motion=Free
  twistLimitDegrees=45
  twistMotion=Free
 }
 {bone1Name=dyn_arm_strap_01_r, bone2Name=dyn_arm_strap_01_r_MSTR, swing1LimitDegrees=25, swing2LimitDegrees=25, twistLimitDegrees=5}
 {bone1Name=dyn_arm_strap_02_r, bone2Name=dyn_arm_strap_01_r, swing1LimitDegrees=25, swing2LimitDegrees=25, twistLimitDegrees=5}
 {
  bone1Name=dyn_torso_zipper_l_MSTR
  bone2Name=clavicle_l
  swing1LimitDegrees=45
  swing1Motion=Free
  swing2LimitDegrees=45
  swing2Motion=Free
  twistLimitDegrees=45
  twistMotion=Free
 }
 {
  bone1Name=dyn_torso_zipper_r_MSTR
  bone2Name=clavicle_r
  swing1LimitDegrees=45
  swing1Motion=Free
  swing2LimitDegrees=45
  swing2Motion=Free
  twistLimitDegrees=45
  twistMotion=Free
 }
 {bone1Name=dyn_torso_zipper_l, bone2Name=dyn_torso_zipper_l_MSTR, swing1LimitDegrees=25, swing2LimitDegrees=25, twistLimitDegrees=5}
 {bone1Name=dyn_torso_zipper_r, bone2Name=dyn_torso_zipper_r_MSTR, swing1LimitDegrees=25, swing2LimitDegrees=25, twistLimitDegrees=5}
]
