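# Character-part AnimBP for the Crazy Eight Tech medium male body, adding secondary motion to straps and zippers.
# It copies the parent mesh pose, places bones with weighted constraints and crouch/thigh-angle offsets, then runs AnimDynamics and per-region RigidBody simulations.
# After simulation, front straps blend toward thigh rotation by movement state.
# The EventGraph computes those alphas, softens most RigidBody nodes while skydiving, and supplies gravity through a deprecated function.
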AnimBlueprint /BRCosmetics/Characters/Player/Male/Medium/Bodies/M_MED_Crazy_Eight_Tech/Meshes/M_MED_Crazy_Eight_Tech_AnimBP
target_skeleton=&/M_MED_Crazy_Eight_Tech_Skeleton
graphs [
 AnimGraph {nodes [
   n1 = Get F Strap Const Alpha() -> (F Strap Const Alpha:double)
   n2 = "Constraint - dyn_front_strap_03_r (R)"(ComponentPose=Local To Component(LocalPose=Use cached pose 'PostSim'()), Alpha=@n1, ConstraintWeights_0=1) {
    alphaBoolBlend {blendOption=Linear}
    alphaScaleBiasClamp {bInterpResult=true}
    boneToModify {boneName=dyn_front_strap_03_r}
    constraintSetup [{transformType=Rotation, targetBone {boneName=thigh_r}}]
    constraintWeights [1]
   }
   n3 = "Constraint - dyn_front_strap_04_r (R)"(ComponentPose=@n2, Alpha=@n1, ConstraintWeights_0=1) {
    alphaBoolBlend {blendOption=Linear}
    alphaScaleBiasClamp {bInterpResult=true}
    boneToModify {boneName=dyn_front_strap_04_r}
    constraintSetup [{transformType=Rotation, targetBone {boneName=thigh_r}}]
    constraintWeights [1]
   }
   n4 = "Constraint - dyn_front_strap_05_r (R)"(ComponentPose=@n3, Alpha=@n1, ConstraintWeights_0=1) {
    alphaBoolBlend {blendOption=Linear}
    alphaScaleBiasClamp {bInterpResult=true}
    boneToModify {boneName=dyn_front_strap_05_r}
    constraintSetup [{transformType=Rotation, targetBone {boneName=thigh_r}}]
    constraintWeights [1]
   }
   n5 = "Constraint - dyn_front_strap_06_r (R)"(ComponentPose=@n4, Alpha=@n1, ConstraintWeights_0=1) {
    alphaBoolBlend {blendOption=Linear}
    alphaScaleBiasClamp {bInterpResult=true}
    boneToModify {boneName=dyn_front_strap_06_r}
    constraintSetup [{transformType=Rotation, targetBone {boneName=thigh_r}}]
    constraintWeights [1]
   }
   n6 = "Constraint - dyn_front_strap_07_r (R)"(ComponentPose=@n5, Alpha=@n1, ConstraintWeights_0=1) {
    alphaBoolBlend {blendOption=Linear}
    alphaScaleBiasClamp {bInterpResult=true, bMapRange=true, outRange {max=0.35}}
    boneToModify {boneName=dyn_front_strap_07_r}
    constraintSetup [{transformType=Rotation, targetBone {boneName=thigh_r}}]
    constraintWeights [1]
   }
   n7 = "Constraint - dyn_front_strap_08_r (R)"(ComponentPose=@n6, Alpha=@n1, ConstraintWeights_0=1) {
    alphaBoolBlend {blendOption=Linear}
    alphaScaleBiasClamp {bInterpResult=true, bMapRange=true, outRange {max=0.35}}
    boneToModify {boneName=dyn_front_strap_08_r}
    constraintSetup [{transformType=Rotation, targetBone {boneName=thigh_r}}]
    constraintWeights [1]
   }
   n8 = "Constraint - dyn_front_strap_09_r (R)"(ComponentPose=@n7, Alpha=@n1, ConstraintWeights_0=1) {
    alphaBoolBlend {blendOption=Linear}
    alphaScaleBiasClamp {bInterpResult=true, bMapRange=true, outRange {max=0.35}}
    boneToModify {boneName=dyn_front_strap_09_r}
    constraintSetup [{transformType=Rotation, targetBone {boneName=thigh_r}}]
    constraintWeights [1]
   }
   n9 = Get F Strap Const Alpha()
   n10 = "Constraint - dyn_front_strap_03_l (R)"(ComponentPose=@n8, Alpha=@n9, ConstraintWeights_0=1) {
    alphaBoolBlend {blendOption=Linear}
    alphaScaleBiasClamp {bInterpResult=true}
    boneToModify {boneName=dyn_front_strap_03_l}
    constraintSetup [{transformType=Rotation, targetBone {boneName=thigh_l}}]
    constraintWeights [1]
   }
   n11 = "Constraint - dyn_front_strap_04_l (R)"(ComponentPose=@n10, Alpha=@n9, ConstraintWeights_0=1) {
    alphaBoolBlend {blendOption=Linear}
    alphaScaleBiasClamp {bInterpResult=true}
    boneToModify {boneName=dyn_front_strap_04_l}
    constraintSetup [{transformType=Rotation, targetBone {boneName=thigh_l}}]
    constraintWeights [1]
   }
   n12 = "Constraint - dyn_front_strap_05_l (R)"(ComponentPose=@n11, Alpha=@n9, ConstraintWeights_0=1) {
    alphaBoolBlend {blendOption=Linear}
    alphaScaleBiasClamp {bInterpResult=true}
    boneToModify {boneName=dyn_front_strap_05_l}
    constraintSetup [{transformType=Rotation, targetBone {boneName=thigh_l}}]
    constraintWeights [1]
   }
   n13 = "Constraint - dyn_front_strap_06_l (R)"(ComponentPose=@n12, Alpha=@n9, ConstraintWeights_0=1) {
    alphaBoolBlend {blendOption=Linear}
    alphaScaleBiasClamp {bInterpResult=true, bMapRange=true, outRange {max=0.35}}
    boneToModify {boneName=dyn_front_strap_06_l}
    constraintSetup [{transformType=Rotation, targetBone {boneName=thigh_l}}]
    constraintWeights [1]
   }
   n14 = "Constraint - dyn_front_strap_07_l (R)"(ComponentPose=@n13, Alpha=@n9, ConstraintWeights_0=1) {
    alphaBoolBlend {blendOption=Linear}
    alphaScaleBiasClamp {bInterpResult=true, bMapRange=true, outRange {max=0.35}}
    boneToModify {boneName=dyn_front_strap_07_l}
    constraintSetup [{transformType=Rotation, targetBone {boneName=thigh_l}}]
    constraintWeights [1]
   }
   n15 = "Constraint - dyn_front_strap_08_l (R)"(ComponentPose=@n14, Alpha=@n9, ConstraintWeights_0=1) {
    alphaBoolBlend {blendOption=Linear}
    alphaScaleBiasClamp {bInterpResult=true, bMapRange=true, outRange {max=0.35}}
    boneToModify {boneName=dyn_front_strap_08_l}
    constraintSetup [{transformType=Rotation, targetBone {boneName=thigh_l}}]
    constraintWeights [1]
   }
   Output Pose(Result=Component To Local(ComponentPose=@n15)) {layerGroup=None}
   n16 = "Constraint - dyn_torso_zipper_l_MSTR (PPPPPPPPP)"!(
    ComponentPose=Local To Component(LocalPose=Copy Pose From Mesh() {bUseAttachedParent=true})
    Alpha=1
    ConstraintWeights_0=0.0002
    ConstraintWeights_1:float=0.0616
    ConstraintWeights_2:float=0.1023
    ConstraintWeights_3:float=0.1827
    ConstraintWeights_4:float=0.4267
    ConstraintWeights_5:float=0.218
    ConstraintWeights_6:float=0.0017
    ConstraintWeights_7:float=0.0067
    ConstraintWeights_8:float=0.0001
   ) -> (Pose) {
    alphaBoolBlend {blendOption=Linear}
    boneToModify {boneName=dyn_torso_zipper_l_MSTR}
    constraintSetup [
     {targetBone {boneName=spine_02}}
     {targetBone {boneName=spine_03}}
     {targetBone {boneName=spine_04}}
     {targetBone {boneName=spine_05}}
     {targetBone {boneName=clavicle_l}}
     {targetBone {boneName=upperarm_twist_01_l}}
     {targetBone {boneName=neck_01}}
     {targetBone {boneName=neck_02}}
     {targetBone {boneName=head}}
    ]
    constraintWeights [1,1,1,1,1,1,1,1,1]
   }
   n17 = "Constraint - dyn_torso_zipper_r_MSTR (PPPPPPPPP)"!(
    ComponentPose=@n16
    Alpha=1
    ConstraintWeights_0=0.0002
    ConstraintWeights_1:float=0.0621
    ConstraintWeights_2:float=0.1047
    ConstraintWeights_3:float=0.1833
    ConstraintWeights_4:float=0.4297
    ConstraintWeights_5:float=0.2119
    ConstraintWeights_6:float=0.0016
    ConstraintWeights_7:float=0.0064
    ConstraintWeights_8:float=0.0001
   ) -> (Pose) {
    alphaBoolBlend {blendOption=Linear}
    boneToModify {boneName=dyn_torso_zipper_r_MSTR}
    constraintSetup [
     {targetBone {boneName=spine_02}}
     {targetBone {boneName=spine_03}}
     {targetBone {boneName=spine_04}}
     {targetBone {boneName=spine_05}}
     {targetBone {boneName=clavicle_r}}
     {targetBone {boneName=upperarm_twist_01_r}}
     {targetBone {boneName=neck_01}}
     {targetBone {boneName=neck_02}}
     {targetBone {boneName=head}}
    ]
    constraintWeights [1,1,1,1,1,1,1,1,1]
   }
   n18 = "Constraint - dyn_arm_strap_01_l_MSTR (PPPPPPPPP)"!(
    ComponentPose=@n17
    Alpha=1
    ConstraintWeights_0=0.0005
    ConstraintWeights_1:float=0
    ConstraintWeights_2:float=0.0001
    ConstraintWeights_3:float=0.0224
    ConstraintWeights_4:float=0.3812
    ConstraintWeights_5:float=0.0041
    ConstraintWeights_6:float=0.0001
    ConstraintWeights_7:float=0.2947
    ConstraintWeights_8:float=0.297
   ) -> (Pose) {
    alphaBoolBlend {blendOption=Linear}
    boneToModify {boneName=dyn_arm_strap_01_l_MSTR}
    constraintSetup [
     {targetBone {boneName=spine_03}}
     {targetBone {boneName=spine_04}}
     {targetBone {boneName=spine_05}}
     {targetBone {boneName=clavicle_l}}
     {targetBone {boneName=upperarm_l}}
     {targetBone {boneName=lowerarm_l}}
     {targetBone {boneName=lowerarm_twist_02_l}}
     {targetBone {boneName=upperarm_twist_01_l}}
     {targetBone {boneName=upperarm_twist_02_l}}
    ]
    constraintWeights [1,1,1,1,1,1,1,1,1]
   }
   n19 = "Constraint - dyn_arm_strap_01_r_MSTR (PPPPPPPPP)"!(
    ComponentPose=@n18
    Alpha=1
    ConstraintWeights_0=0.0006
    ConstraintWeights_1:float=0
    ConstraintWeights_2:float=0.0001
    ConstraintWeights_3:float=0.023
    ConstraintWeights_4:float=0.3797
    ConstraintWeights_5:float=0.0042
    ConstraintWeights_6:float=0.0001
    ConstraintWeights_7:float=0.2969
    ConstraintWeights_8:float=0.2954
   ) -> (Pose) {
    alphaBoolBlend {blendOption=Linear}
    boneToModify {boneName=dyn_arm_strap_01_r_MSTR}
    constraintSetup [
     {targetBone {boneName=spine_03}}
     {targetBone {boneName=spine_04}}
     {targetBone {boneName=spine_05}}
     {targetBone {boneName=clavicle_r}}
     {targetBone {boneName=upperarm_r}}
     {targetBone {boneName=lowerarm_r}}
     {targetBone {boneName=lowerarm_twist_02_r}}
     {targetBone {boneName=upperarm_twist_01_r}}
     {targetBone {boneName=upperarm_twist_02_r}}
    ]
    constraintWeights [1,1,1,1,1,1,1,1,1]
   }
   n20 = "Constraint - dyn_torso_strap_l (PPPPPPPPP)"!(
    ComponentPose=@n19
    Alpha=1
    ConstraintWeights_0=0.0448
    ConstraintWeights_1:float=0.158
    ConstraintWeights_2:float=0.237
    ConstraintWeights_3:float=0.2615
    ConstraintWeights_4:float=0.1552
    ConstraintWeights_5:float=0.0331
    ConstraintWeights_6:float=0.0814
    ConstraintWeights_7:float=0.0087
    ConstraintWeights_8:float=0.0204
   ) -> (Pose) {
    alphaBoolBlend {blendOption=Linear}
    boneToModify {boneName=dyn_torso_strap_l}
    constraintSetup [
     {targetBone {boneName=pelvis}}
     {targetBone {boneName=spine_01}}
     {targetBone {boneName=spine_02}}
     {targetBone {boneName=spine_03}}
     {targetBone {boneName=spine_04}}
     {targetBone {boneName=spine_05}}
     {targetBone {boneName=upperarm_twist_01_l}}
     {targetBone {boneName=dyn_back_strap_01_l}}
     {targetBone {boneName=dyn_front_strap_01_l}}
    ]
    constraintWeights [1,1,1,1,1,1,1,1,1]
   }
   n21 = "Constraint - dyn_torso_strap_r (PPPPPPPP)"!(
    ComponentPose=@n20
    Alpha=1
    ConstraintWeights_0=0.0445
    ConstraintWeights_1:float=0.1365
    ConstraintWeights_2:float=0.2358
    ConstraintWeights_3:float=0.2577
    ConstraintWeights_4:float=0.1525
    ConstraintWeights_5:float=0.0314
    ConstraintWeights_6:float=0.0762
    ConstraintWeights_7:float=0.0654
   ) -> (Pose) {
    alphaBoolBlend {blendOption=Linear}
    boneToModify {boneName=dyn_torso_strap_r}
    constraintSetup [
     {targetBone {boneName=pelvis}}
     {targetBone {boneName=spine_01}}
     {targetBone {boneName=spine_02}}
     {targetBone {boneName=spine_03}}
     {targetBone {boneName=spine_04}}
     {targetBone {boneName=spine_05}}
     {targetBone {boneName=upperarm_twist_01_r}}
     {targetBone {boneName=dyn_front_strap_01_r}}
    ]
    constraintWeights [1,1,1,1,1,1,1,1]
   }
   n22 = "Constraint - dyn_front_strap_01_r (P)"(ComponentPose=@n21, Alpha=1, ConstraintWeights_0=1) {
    alphaBoolBlend {blendOption=Linear}
    boneToModify {boneName=dyn_front_strap_01_r}
    constraintSetup [{targetBone {boneName=pelvis}}]
    constraintWeights [1]
   }
   n23 = "Constraint - dyn_front_strap_02_r (P)"(ComponentPose=@n22, Alpha=1, ConstraintWeights_0=1) {
    alphaBoolBlend {blendOption=Linear}
    boneToModify {boneName=dyn_front_strap_02_r}
    constraintSetup [{targetBone {boneName=pelvis}}]
    constraintWeights [1]
   }
   n24 = "Constraint - dyn_front_strap_01_l (P)"(ComponentPose=@n23, Alpha=1, ConstraintWeights_0=1) {
    alphaBoolBlend {blendOption=Linear}
    boneToModify {boneName=dyn_front_strap_01_l}
    constraintSetup [{targetBone {boneName=pelvis}}]
    constraintWeights [1]
   }
   n25 = "Constraint - dyn_front_strap_02_l (P)"(ComponentPose=@n24, Alpha=1, ConstraintWeights_0=1) {
    alphaBoolBlend {blendOption=Linear}
    boneToModify {boneName=dyn_front_strap_02_l}
    constraintSetup [{targetBone {boneName=pelvis}}]
    constraintWeights [1]
   }
   n26 = Get bIsCrouching()
   n27 = "Transform (Modify) Bone - Bone: dyn_front_strap_01_r"!(ComponentPose=@n25, bAlphaBoolEnabled:bool=@n26, Translation="2,-2,0", Rotation="-12,0,0", Scale="1,1,1") -> (Pose) {
    alphaInputType=Bool
    rotationMode=BMM_Additive
    rotationSpace=BCS_BoneSpace
    translationMode=BMM_Additive
    translationSpace=BCS_BoneSpace
    alphaBoolBlend {blendOption=Linear}
    boneToModify {boneName=dyn_front_strap_01_r}
    rotation {pitch=-12}
    translation=<2 -2>
   }
   n28 = "Transform (Modify) Bone - Bone: dyn_front_strap_01_l"!(ComponentPose=@n27, bAlphaBoolEnabled:bool=@n26, Translation="2,-2,0", Rotation="10,0,0", Scale="1,1,1") -> (Pose) {
    alphaInputType=Bool
    rotationMode=BMM_Additive
    rotationSpace=BCS_BoneSpace
    translationMode=BMM_Additive
    translationSpace=BCS_BoneSpace
    alphaBoolBlend {blendOption=Linear}
    boneToModify {boneName=dyn_front_strap_01_l}
    rotation {pitch=10}
    translation=<2 -2>
   }
   n29 = "Transform (Modify) Bone - Bone: dyn_front_strap_01_r"!(ComponentPose=@n28, Alpha:float=Get Thigh R Alpha() -> (Thigh R Alpha:double), Translation="1,-1,0", Rotation="-2.5,0,0", Scale="1,1,1") -> (Pose) {
    rotationMode=BMM_Additive
    rotationSpace=BCS_BoneSpace
    translationMode=BMM_Additive
    translationSpace=BCS_BoneSpace
    alphaBoolBlend {blendOption=Linear}
    boneToModify {boneName=dyn_front_strap_01_r}
    rotation {pitch=-2.5}
    translation=<1 -1>
   }
   n30 = "Transform (Modify) Bone - Bone: dyn_front_strap_01_l"!(ComponentPose=@n29, Alpha:float=Get Thigh L Alpha() -> (Thigh L Alpha:double), Translation="0,-1,2", Rotation="2.5,0,0", Scale="1,1,1") -> (Pose) {
    rotationMode=BMM_Additive
    rotationSpace=BCS_BoneSpace
    translationMode=BMM_Additive
    translationSpace=BCS_BoneSpace
    alphaBoolBlend {blendOption=Linear}
    boneToModify {boneName=dyn_front_strap_01_l}
    rotation {pitch=2.5}
    translation {y=-1, z=2}
   }
   Save cached pose 'PreSim'(Pose=Component To Local(ComponentPose=@n30)) {cacheName=PreSim}
   n31 = Anim Dynamics - Bone: dyn_strap_01_l(ComponentPose=Local To Component(LocalPose=Use cached pose 'PreSim'()), Alpha=1, ExternalForce="0,0,0") {
    angularDampingOverride=0.99
    angularSpringConstant=10
    bAngularSpring=true
    bGravityOverrideInSimSpace=true
    bUseGravityOverride=true
    simulationSpace=RootRelative
    alphaBoolBlend {blendOption=Linear}
    boundBone {boneName=dyn_strap_01_l}
    componentAppliedLinearAccClamp=<3000 3000 3000>
    componentLinearAccScale=<0.1 0.1 0.1>
    componentLinearVelScale=<0.9 0.9 0.9>
    gravityOverride {z=-980}
    physicsBodyDefinitions [{
      sphereCollisionRadius=0
      boundBone {boneName=dyn_strap_01_l}
      boxExtents=<12 8 12>
      constraintSetup {angularTargetAxis=AxisZ, angularLimitsMax=<0 5 0>, angularLimitsMin=<0 -5 -10>, angularTarget=<0 0 1>}
      localJointOffset=<2.5 0 0>
     }]
    relativeSpaceBone {boneName=thigh_l}
   }
   n32 = Anim Dynamics - Bone: dyn_strap_02_l(ComponentPose=@n31, Alpha=1, ExternalForce="0,0,0") {
    angularDampingOverride=0.99
    angularSpringConstant=10
    bAngularSpring=true
    bGravityOverrideInSimSpace=true
    bOverrideAngularDamping=true
    bUseGravityOverride=true
    simulationSpace=RootRelative
    alphaBoolBlend {blendOption=Linear}
    boundBone {boneName=dyn_strap_02_l}
    componentAppliedLinearAccClamp=<500 500 500>
    componentLinearAccScale=<0.1 0.1 0.1>
    componentLinearVelScale=<0.9 0.9 0.9>
    gravityOverride {z=-980}
    physicsBodyDefinitions [{
      sphereCollisionRadius=0
      boundBone {boneName=dyn_strap_02_l}
      boxExtents=<12 8 12>
      constraintSetup {angularTargetAxis=AxisZ, angularLimitsMax=<0 5 0>, angularLimitsMin=<0 -5 -20>, angularTarget=<0 0 1>}
      localJointOffset=<2.5 0 0>
     }]
    relativeSpaceBone {boneName=thigh_l}
   }
   n33 = Anim Dynamics - Bone: dyn_arm_strap_01_l(ComponentPose=@n32, Alpha=1, ExternalForce="0,0,0") {
    angularDampingOverride=0.999
    angularSpringConstant=10
    bAngularSpring=true
    bGravityOverrideInSimSpace=true
    bUseGravityOverride=true
    simulationSpace=RootRelative
    alphaBoolBlend {blendOption=Linear}
    boundBone {boneName=dyn_arm_strap_01_l}
    componentAppliedLinearAccClamp=<3000 3000 3000>
    componentLinearAccScale=<0.1 0.1 0.1>
    componentLinearVelScale=<0.9 0.9 0.9>
    gravityOverride {z=-980}
    physicsBodyDefinitions [{
      sphereCollisionRadius=0
      boundBone {boneName=dyn_arm_strap_01_l}
      boxExtents=<6 3 6>
      constraintSetup {angularTargetAxis=AxisZ, angularLimitsMax=<2 7 15>, angularLimitsMin=<-2 -7 -15>, angularTarget=<0 0 1>}
      localJointOffset=<2.5 0 0>
     }]
    relativeSpaceBone {boneName=dyn_arm_strap_01_l_MSTR}
   }
   n34 = Anim Dynamics - Bone: dyn_arm_strap_01_r(ComponentPose=@n33, Alpha=1, ExternalForce="0,0,0") {
    angularDampingOverride=0.999
    angularSpringConstant=10
    bAngularSpring=true
    bGravityOverrideInSimSpace=true
    bUseGravityOverride=true
    simulationSpace=RootRelative
    alphaBoolBlend {blendOption=Linear}
    boundBone {boneName=dyn_arm_strap_01_r}
    componentAppliedLinearAccClamp=<3000 3000 3000>
    componentLinearAccScale=<0.1 0.1 0.1>
    componentLinearVelScale=<0.9 0.9 0.9>
    gravityOverride {z=-980}
    physicsBodyDefinitions [{
      sphereCollisionRadius=0
      boundBone {boneName=dyn_arm_strap_01_r}
      boxExtents=<6 3 6>
      constraintSetup {angularTargetAxis=AxisZ, angularLimitsMax=<2 7 15>, angularLimitsMin=<-2 -7 -15>, angularTarget=<0 0 1>}
      localJointOffset=<2.5 0 0>
     }]
    relativeSpaceBone {boneName=dyn_arm_strap_01_r_MSTR}
   }
   n35 = Anim Dynamics - Bone: dyn_arm_strap_02_l(ComponentPose=@n34, Alpha=1, ExternalForce="0,0,0") {
    angularDampingOverride=0.999
    angularSpringConstant=10
    bAngularSpring=true
    bGravityOverrideInSimSpace=true
    bUseGravityOverride=true
    simulationSpace=RootRelative
    alphaBoolBlend {blendOption=Linear}
    boundBone {boneName=dyn_arm_strap_02_l}
    componentAppliedLinearAccClamp=<3000 3000 3000>
    componentLinearAccScale=<0.1 0.1 0.1>
    componentLinearVelScale=<0.9 0.9 0.9>
    gravityOverride {z=-980}
    physicsBodyDefinitions [{
      sphereCollisionRadius=0
      boundBone {boneName=dyn_arm_strap_02_l}
      boxExtents=<6 3 6>
      constraintSetup {angularTargetAxis=AxisZ, angularLimitsMax=<0 0 15>, angularLimitsMin=<0 0 -15>, angularTarget=<0 0 1>}
      localJointOffset=<2.5 0 0>
     }]
    relativeSpaceBone {boneName=dyn_arm_strap_01_l}
   }
   n36 = Anim Dynamics - Bone: dyn_arm_strap_02_r(ComponentPose=@n35, Alpha=1, ExternalForce="0,0,0") {
    angularDampingOverride=0.999
    angularSpringConstant=10
    bAngularSpring=true
    bGravityOverrideInSimSpace=true
    bUseGravityOverride=true
    simulationSpace=RootRelative
    alphaBoolBlend {blendOption=Linear}
    boundBone {boneName=dyn_arm_strap_02_r}
    componentAppliedLinearAccClamp=<3000 3000 3000>
    componentLinearAccScale=<0.1 0.1 0.1>
    componentLinearVelScale=<0.9 0.9 0.9>
    gravityOverride {z=-980}
    physicsBodyDefinitions [{
      sphereCollisionRadius=0
      boundBone {boneName=dyn_arm_strap_02_r}
      boxExtents=<6 3 6>
      constraintSetup {angularTargetAxis=AxisZ, angularLimitsMax=<0 0 15>, angularLimitsMin=<0 0 -15>, angularTarget=<0 0 1>}
      localJointOffset=<2.5 0 0>
     }]
    relativeSpaceBone {boneName=dyn_arm_strap_01_r}
   }
   n37 = Get GravityOverride()
   n38 = RigidBody!(ComponentPose=@n36, Alpha=Get Skydive Alpha() -> (Skydive Alpha:double), OverrideWorldGravity:Vector=@n37, ExternalForce) -> (Pose) {
    bClampLinearTranslationLimitToRefPose=true
    bOverrideWorldGravity=true
    overridePhysicsAsset=&/M_MED_Crazy_Eight_Tech_Physics
    simulationSpace=BaseBoneSpace
    alphaBoolBlend {blendOption=Linear}
    baseBoneRef {boneName=pelvis}
    componentLinearAccScale=<0.3 0.3 0.3>
    componentLinearVelScale=<0.3 0.3 0.3>
    overrideWorldGravity {z=-980}
   }
   n39 = RigidBody(ComponentPose=@n38, OverrideWorldGravity=@n37) {
    bClampLinearTranslationLimitToRefPose=true
    bOverrideWorldGravity=true
    overridePhysicsAsset=&/M_MED_Crazy_Eight_Tech_Torso_Physics
    simulationSpace=BaseBoneSpace
    alphaBoolBlend {blendOption=Linear}
    baseBoneRef {boneName=spine_05}
    componentLinearAccScale=<0.2 0.2 0.2>
    componentLinearVelScale=<0.2 0.2 0.2>
    overrideWorldGravity {z=-980}
   }
   n40 = Get Skydive Alpha()
   n41 = RigidBody(ComponentPose=@n39, Alpha=@n40, OverrideWorldGravity=@n37) {
    bClampLinearTranslationLimitToRefPose=true
    bOverrideWorldGravity=true
    overridePhysicsAsset=&/M_MED_Crazy_Eight_Tech_RLeg_Physics
    simulationSpace=BaseBoneSpace
    alphaBoolBlend {blendOption=Linear}
    baseBoneRef {boneName=thigh_r}
    componentLinearAccScale=<0.3 0.3 0.3>
    componentLinearVelScale=<0.3 0.3 0.3>
    overrideWorldGravity {z=-980}
   }
   n42 = RigidBody(ComponentPose=@n41, Alpha=@n40, OverrideWorldGravity=@n37) {
    bClampLinearTranslationLimitToRefPose=true
    bOverrideWorldGravity=true
    overridePhysicsAsset=&/M_MED_Crazy_Eight_Tech_LLeg_Physics
    simulationSpace=BaseBoneSpace
    alphaBoolBlend {blendOption=Linear}
    baseBoneRef {boneName=thigh_l}
    componentLinearAccScale=<0.3 0.3 0.3>
    componentLinearVelScale=<0.3 0.3 0.3>
    overrideWorldGravity {z=-980}
   }
   Save cached pose 'PostSim'(Pose=Component To Local(ComponentPose=@n42)) {cacheName=PostSim}
  ]}
 EventGraph {nodes [
   Event BlueprintUpdateAnimation()
   > n1 = Sequence!(execute) -> (then_0, then_1, then_2:Exec, then_3:Exec)
   > n2 = Gravity Override State Function DEPRECATED(CharacterPartAnimInstance=Self-Reference(), GameTimeInSeconds=GetGameTimeInSeconds(), GravityOverrideParams=Get Gravity Override Params())
   > Set GravityOverride(GravityOverride=@n2)
   n3 = Branch(execute=@n1.then_1, Condition=Get bIsSkyDiving())
   > Set Skydive Alpha(Skydive Alpha:double=0.6) -> (Output_Get:double)
   Set Skydive Alpha(execute=@n3.else, Skydive Alpha=1)
   n4 = Branch(execute=@n1.then_2, Condition=Get bIsCrouching())
   > Set F Strap Const Alpha(F Strap Const Alpha:double=0) -> (Output_Get:double)
   n5 = Branch(execute=@n4.else, Condition="float > float"(A=Get PawnSpeed2D(), B=0) {is=title K2Node_CallFunction})
   > n6 = Branch(Condition=Get bIsSprinting())
   > Set F Strap Const Alpha(F Strap Const Alpha=1)
   Set F Strap Const Alpha(execute=@n6.else, F Strap Const Alpha=0.4)
   Set F Strap Const Alpha(execute=@n5.else, F Strap Const Alpha=0)
   n7 = BreakTransform(InTransform=GetDeltaTransformFromRefPose(self=GetOwnerSkeletalMeshComponent(), BoneName=thigh_r, BaseName=pelvis))
   Set Thigh R Alpha(execute=@n1.then_3, Thigh R Alpha:double=MapRangeClamped(Value=BreakRotator(InRot=@n7.Rotation).Yaw, InRangeA=0, InRangeB=-20)) -> (Output_Get:double)
   n8 = BreakTransform(InTransform=GetDeltaTransformFromRefPose(self=GetOwnerSkeletalMeshComponent(), BoneName=thigh_l, BaseName=pelvis))
   > Set Thigh L Alpha(Thigh L Alpha:double=MapRangeClamped(Value=BreakRotator(InRot=@n8.Rotation).Yaw, InRangeA=0, InRangeB=-20)) -> (Output_Get:double)
  ]}
]
variables [
 F Strap Const Alpha {type=number, default=0.4}
 Gravity Override Params {type=GravityOverrideParamsStruct}
 GravityOverride {type=Vector}
 Skydive Alpha {type=number}
 Thigh L Alpha {type=number}
 Thigh R Alpha {type=number}
]
