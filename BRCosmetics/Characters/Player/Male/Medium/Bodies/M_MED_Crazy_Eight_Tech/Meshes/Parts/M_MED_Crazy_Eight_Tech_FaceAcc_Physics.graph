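# Physics asset for the Crazy Eight Tech male medium body's face accessory, covering the left and right dynamic hood bones.
# Each side chains a small master capsule, angularly free relative to the head, to a flat box body held by tight swing and twist limits.
# Capsules on the head and both upper-arm twist bones have no constraints between them, apparently serving as colliders for the hood.
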PhysicsAsset /BRCosmetics/Characters/Player/Male/Medium/Bodies/M_MED_Crazy_Eight_Tech/Meshes/Parts/M_MED_Crazy_Eight_Tech_FaceAcc_Physics
bodies {
 dyn_hood_L_01 [{shapeType=Box, extentX=10, extentY=10, extentZ=1.01, shapeName=dyn_hood_L_01_box, center=<0 4.3679 0>}]
 dyn_hood_L_MSTR [{shapeType=Capsule, length=0.505, radius=0.505, shapeName=dyn_hood_L_MSTR_capsule, rotation=<rot 0 -90 0>}]
 dyn_hood_R_01 [{shapeType=Box, extentX=10, extentY=10, extentZ=1.01, shapeName=dyn_hood_R_01_box, center=<0 4.3679 0>}]
 dyn_hood_R_MSTR [{shapeType=Capsule, length=0.505, radius=0.505, shapeName=dyn_hood_R_MSTR_capsule, rotation=<rot 0 -90 0>}]
 head [{shapeType=Capsule, length=11.1049, radius=9.0545, center=<4.0418 3.3947 0.2235>, rotation=<rot 90 -14.0363 -19.0374>}]
 upperarm_twist_01_l [{shapeType=Capsule, length=25, radius=10, shapeName=upperarm_twist_01_l_capsule, center=<15 0 0>, rotation=<rot 90 0 0>}]
 upperarm_twist_01_r [{shapeType=Capsule, length=25, radius=10, shapeName=upperarm_twist_01_r_capsule, center=<-15 0 0>, rotation=<rot 90 0 0>}]
}
constraints [
 {
  bone1Name=dyn_hood_R_MSTR
  bone2Name=head
  swing1LimitDegrees=45
  swing1Motion=Free
  swing2LimitDegrees=45
  swing2Motion=Free
  twistLimitDegrees=45
  twistMotion=Free
 }
 {
  bone1Name=dyn_hood_L_MSTR
  bone2Name=head
  swing1LimitDegrees=45
  swing1Motion=Free
  swing2LimitDegrees=45
  swing2Motion=Free
  twistLimitDegrees=45
  twistMotion=Free
 }
 {bone1Name=dyn_hood_L_01, bone2Name=dyn_hood_L_MSTR, swing1LimitDegrees=8, swing2LimitDegrees=3, twistLimitDegrees=15}
 {bone1Name=dyn_hood_R_01, bone2Name=dyn_hood_R_MSTR, swing1LimitDegrees=8, swing2LimitDegrees=3, twistLimitDegrees=15}
]
