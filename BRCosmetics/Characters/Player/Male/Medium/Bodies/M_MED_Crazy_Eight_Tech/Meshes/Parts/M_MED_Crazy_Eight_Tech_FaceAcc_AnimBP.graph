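# Post-process style animation Blueprint for the face accessory part of the male Crazy Eight Tech character, adding secondary physics motion and feeding eye position to its materials.
# The AnimGraph copies the attached parent's pose, then runs a RigidBody simulation relative to the head bone, using its own physics asset, overridden gravity and damped component motion.
# On update, after a validity check and zero-length delay, it writes the MI Eye socket's component-space location to the Eye material vector parameter.
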AnimBlueprint /BRCosmetics/Characters/Player/Male/Medium/Bodies/M_MED_Crazy_Eight_Tech/Meshes/Parts/M_MED_Crazy_Eight_Tech_FaceAcc_AnimBP
target_skeleton=&/M_MED_Crazy_Eight_Tech_FaceAcc_Skeleton
graphs [
 AnimGraph {nodes [
   n1 = RigidBody(ComponentPose=Local To Component(LocalPose=Copy Pose From Mesh() {bUseAttachedParent=true})) {
    bClampLinearTranslationLimitToRefPose=true
    bOverrideWorldGravity=true
    overridePhysicsAsset=&/M_MED_Crazy_Eight_Tech_FaceAcc_Physics
    simulationSpace=BaseBoneSpace
    alphaBoolBlend {blendOption=Linear}
    baseBoneRef {boneName=head}
    componentAppliedLinearAccClamp=<3000 3000 3000>
    componentLinearAccScale=<0.3 0.3 0.3>
    componentLinearVelScale=<0.1 0.1 0.1>
    overrideWorldGravity {z=-980}
   }
   Output Pose(Result=Component To Local(ComponentPose=@n1)) {layerGroup=None}
  ]}
 EventGraph {nodes [
   Event BlueprintUpdateAnimation()
   n1 = GetPartSkeletalMeshComponent()
   > Is Valid(InputObject=@n1)
   > Delay(Duration=0)
   > SetVectorParameterValueOnMaterials!(
    execute
    self:Mesh Component Object Reference=@n1
    ParameterName:Name=Eye
    ParameterValue:Vector=TransformLocation!(T:"Transform (by ref)"=GetSocketTransform(self=@n1, InSocketName=MI Eye, TransformSpace=RTS_Component), Location:Vector="0, 0, 0") -> (ReturnValue:Vector)
   ) -> (then)
  ]}
]
variables []
